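; Quest gameplay ability for the Fearless Flight swing quest, crediting the owning player for reaching a target swing count before landing.
; On event activation it ends unless the quest manager's tags show the WestSausage item in inventory.
; It resets the counter, then waits in parallel for SwingingHit events, which increment it, and the tracked landing event, which ends the ability.
; Reaching the target sends a complex custom stat event and ends the ability.
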(event Ability|EventActivateAbilityFromEvent
  (params (EventData :type Gameplay%20Event%20Data%20Structure%20%28by%20ref%29))
  (|CallParentFunction EventData)
  (if (GameplayTags|HasTag (Quests|GetSourceandContextTags (Variables|GAVars|Data|GetOwningPlayerQuestManager)) "(TagName=\"Granted.Athena.WestSausage.InInventory\")" true)
    ;[1 0.2402 0.03963] Initialize Swing tracking...
    (Variables|Default|SetSwingCount 0)
    (Ability|Tasks|WaitGameplayEvent (Variables|Tags|GetTrackedLandingTag)
      (:then
        (Ability|Tasks|WaitGameplayEvent "(TagName=\"Event.Quests.SwingingHit\")"
          (:EventReceived
            (bind _result (Math|Integer|IncrementInt (Variables|Default|GetSwingCount))
              (:%20%20
                (bind _output_get (Variables|Default|SetSwingCount _result))
                (if (>= _output_get (Variables|Default|GetSwingTarget))
                  ;[0.03963 0.9403 1] Max swing count reached!
                  (Quest|SendComplexCustomStatEvent (Variables|GAVars|Data|GetOwningPlayerQuestManager) 0 (Variables|Tags|GetStatEventSourceTags) "(GameplayTags=)" 1)
                  (Ability|EndAbility)))))))
      (:EventReceived
        ;[1 0.09399 0.7665] Target landed...
        (Ability|EndAbility)))
    (else
      (Ability|EndAbility))))
